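# Sound cue for Creative portal matchmaking, named as the sound for a match being initiated.
# A pitched-up UI whoosh plays immediately, mixed with a briefly delayed, volume-enveloped layer that fades out.
# That layer blends the main-menu start-game sound with a slightly later, quieter OK sound.
# It uses the application tray-notification sound class and overrides concurrency to one instance, stopping the oldest.
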SoundCue /Game/Sounds/Creative/Gadgets/Portal/MatchMaking_MatchInitiated_Cue
root=Mixer_1
nodes [
 {id=Delay_0, children [Enveloper_0], properties {delayMax=0.3, delayMin=0.3}}
 {id=Delay_2, children [WavePlayer_4], properties {delayMax=0.1, delayMin=0.1}}
 {
  id=Enveloper_0
  children [Mixer_0]
  properties {
   pitchCurve {editorCurveData {keys [{time=0}]}}
   volumeCurve {editorCurveData {keys [<at 0 1>, {arriveTangent=-0.8, interpMode=RCIM_Cubic, leaveTangent=0.0461, tangentMode=RCTM_Break, time=0.7952, value=0.9158}, <at 1.25 0>]}}
  }
 }
 {id=Mixer_0, children [WavePlayer_0,Delay_2], properties {inputVolume [1.5,0.7]}}
 {id=Mixer_1, children [Modulator_0,Delay_0], properties {inputVolume [1.5,1]}}
 {id=Modulator_0, children [WavePlayer_1], properties {pitchMax=1.2, pitchMin=1.2}}
 {id=WavePlayer_0, properties {soundWaveAssetPtr=/Game/Sounds/Fort_UI_Sounds/FrontEnd/New_FrontEnd/Fortnite_UI_MainMenu_StartGame_01}}
 {id=WavePlayer_1, properties {soundWaveAssetPtr=/Game/Sounds/Fort_UI_Sounds/Gameplay/UI_ChangeHero_Whoosh_RightSide_01}}
 {id=WavePlayer_4, properties {soundWaveAssetPtr=/Game/Sounds/Fort_UI_Sounds/FrontEnd/fort_ui_ok_02}}
]
properties {
 bOverrideConcurrency=true
 maxDistance=2097152
 soundClassObject=/Game/Sounds/FortSoundClassesAndModes/Fort_Application_TrayNotification
 volumeMultiplier=1.3967
 concurrencyOverrides {maxCount=1, resolutionRule=StopOldest}
}
