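# Sound cue for the Creative matchmaking portal gadget, pairing an opening one-shot with a looping hum.
# A mixer blends the open sound, pitched slightly down through a modulator, with the looping wave, whose enveloper fades it in briefly as the opening plays.
# The mix runs through rift portal attenuation with overrides, and a concurrency override caps simultaneous voices with timed release when stolen.
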SoundCue /Game/Sounds/Creative/Gadgets/Portal/MatchMaking_Portal_Loop_Cue
root=Attenuation_1
nodes [
 {
  id=Attenuation_1
  children [Mixer_1]
  properties {
   attenuationSettings=/Game/Sounds/Attenuation/Fort_RiftPortals_Creative_Att
   attenuationOverrides {falloffDistance=60, attenuationShapeExtents {x=0}}
  }
 }
 {
  id=Enveloper_0
  children [WavePlayer_4]
  properties {pitchCurve {editorCurveData {keys [{time=0}]}}, volumeCurve {editorCurveData {keys [<at 0 0>,<at 0.3 1>]}}}
 }
 {id=Mixer_1, children [Modulator_0,Enveloper_0], properties {inputVolume [0.5,0.7]}}
 {id=Modulator_0, children [WavePlayer_1], properties {pitchMax=0.9, pitchMin=0.9}}
 {id=WavePlayer_1, properties {soundWaveAssetPtr=&/MatchMaking_Portal_Open_01}}
 {id=WavePlayer_4, properties {bLooping=true, soundWaveAssetPtr=&/MatchMaking_Portal_Loop_01}}
]
properties {
 attenuationSettings=&Fort_RiftPortals_Creative_Att
 bOverrideConcurrency=true
 maxDistance=1800
 soundClassObject=&Fort_Gadgets_Creative
 volumeMultiplier=0.5
 concurrencyOverrides {bVolumeScaleCanRelease=true, maxCount=3, voiceStealReleaseTime=0.5, volumeScaleAttackTime=0.5}
}
