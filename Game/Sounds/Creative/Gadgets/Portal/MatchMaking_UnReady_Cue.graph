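# Sound cue for the Creative portal gadget, played by its name when matchmaking ready-up is cancelled.
# A single wave player plays the front-end UI ready-up disable sound at reduced volume under the tray notification sound class.
# A concurrency override allows one instance at a time, stopping the oldest, with virtualization disabled and a very large max distance.
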SoundCue /Game/Sounds/Creative/Gadgets/Portal/MatchMaking_UnReady_Cue
root=WavePlayer_0
nodes [{id=WavePlayer_0, properties {soundWaveAssetPtr=/Game/Sounds/Fort_UI_Sounds/FrontEnd/UI_MatchMaking_ReadyUp_Disable_01}}]
properties {
 bOverrideConcurrency=true
 maxDistance=2097152
 soundClassObject=/Game/Sounds/FortSoundClassesAndModes/Fort_Application_TrayNotification
 virtualizationMode=Disabled
 volumeMultiplier=0.5
 concurrencyOverrides {maxCount=1, resolutionRule=StopOldest}
}
