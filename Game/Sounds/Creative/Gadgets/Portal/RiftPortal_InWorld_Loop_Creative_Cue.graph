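# Looping in-world sound cue for the Creative rift portal gadget.
# The first two quality levels mix an oddities loop on the shared portal attenuation with a rift close loop using custom HRTF, listener-focus and occlusion attenuation.
# The other two quality levels play only an "_opt" variant of the oddities loop on that shared attenuation.
# A concurrency override caps simultaneous instances, stopping the farthest first.
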SoundCue /Game/Sounds/Creative/Gadgets/Portal/RiftPortal_InWorld_Loop_Creative_Cue
root=QualityLevel_0
nodes [
 {
  id=Attenuation_0
  children [WavePlayer_3]
  properties {
   bOverrideAttenuation=true
   attenuationOverrides {
    bEnableFocusInterpolation=true
    bEnableListenerFocus=true
    bEnableOcclusion=true
    falloffDistance=876.4608
    focusAzimuth=20
    nonFocusAzimuth=40
    nonFocusVolumeAttenuation=0.3
    occlusionLowPassFilterFrequency=10000
    occlusionTraceChannel=ECC_GameTraceChannel11
    occlusionVolumeAttenuation=0.3
    spatializationAlgorithm=SPATIALIZATION_HRTF
    stereoSpread=959.1758
    attenuationShapeExtents {x=0}
   }
  }
 }
 {
  id=Attenuation_2
  children [WavePlayer_0]
  properties {
   attenuationSettings=/Game/Sounds/Attenuation/Fort_RiftPortals_Creative_Att
   attenuationOverrides {falloffDistance=60, attenuationShapeExtents {x=0}}
  }
 }
 {
  id=Attenuation_3
  children [WavePlayer_2]
  properties {
   attenuationSettings=&Fort_RiftPortals_Creative_Att
   attenuationOverrides {falloffDistance=60, attenuationShapeExtents {x=0}}
  }
 }
 {id=Mixer_0, children [Attenuation_2, Attenuation_0], properties {inputVolume [0.8,1.2]}}
 {id=QualityLevel_0, children [Mixer_0, Mixer_0, Attenuation_3, Attenuation_3]}
 {id=WavePlayer_0, properties {bLooping=true, soundWaveAssetPtr=/Game/Athena/Sounds/Events/Geode/Oddities/Geode_Oddities_InWorldLoop_01}}
 {id=WavePlayer_2, properties {bLooping=true, soundWaveAssetPtr=/Game/Athena/Sounds/Events/Geode/Oddities/Geode_Oddities_InWorldLoop_01_opt}}
 {id=WavePlayer_3, properties {bLooping=true, soundWaveAssetPtr=/Game/Sounds/Creative/Rift/Creative_Rift_CloseLoop_New}}
]
properties {
 bOverrideConcurrency=true
 maxDistance=1800
 soundClassObject=&Fort_Gadgets_Items
 volumeMultiplier=0.6
 concurrencyOverrides {maxCount=2, resolutionRule=StopFarthestThenPreventNew, volumeScaleMode=Distance}
}
