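# Sound cue for the looping in-world pillar sound of the Creative Rift Portal gadget, routed to the gadget items sound class.
# A looping wave feeds an enveloper that ramps volume up over the first few seconds, then custom attenuation with low-pass filtering and listener focus that quiets off-focus sources.
# A mixer with one empty input serves every quality level identically.
# Concurrency caps simultaneous instances at three, stopping the quietest.
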SoundCue /Game/Sounds/Creative/Gadgets/Portal/RiftPortal_InWorld_PillarLoop_Creative_Cue
root=QualityLevel_0
nodes [
 {
  id=Attenuation_0
  children [Enveloper_0]
  properties {
   bOverrideAttenuation=true
   attenuationOverrides {
    bAttenuateWithLPF=true
    bEnableFocusInterpolation=true
    bEnableListenerFocus=true
    falloffDistance=876.4608
    focusAttackInterpSpeed=4
    focusAzimuth=20
    focusReleaseInterpSpeed=4
    nonFocusAzimuth=40
    nonFocusVolumeAttenuation=0.3
    stereoSpread=959.1758
    attenuationShapeExtents {x=478.6349}
   }
  }
 }
 {
  id=Enveloper_0
  children [WavePlayer_1]
  properties {pitchCurve {editorCurveData {keys [{time=0}]}}, volumeCurve {editorCurveData {keys [<at 0 0.2>,<at 3 1>]}}}
 }
 {id=Mixer_0, children [null, Attenuation_0], properties {inputVolume [0.8,1.2]}}
 {id=QualityLevel_0, children [Mixer_0,Mixer_0,Mixer_0,Mixer_0]}
 {id=WavePlayer_1, properties {bLooping=true, soundWaveAssetPtr=&/Creative_Rift_PillarLoop_01}}
]
properties {
 bOverrideConcurrency=true
 maxDistance=1355.0957
 soundClassObject=&Fort_Gadgets_Items
 volumeMultiplier=0.25
 concurrencyOverrides {maxCount=3, resolutionRule=StopQuietest}
}
